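; Blueprint interface for weapon abilities in the AICore common layer, letting implementers override firing ability cooldowns for AI.
; It declares one function, FiringAbilityCooldownOverride, which takes no inputs.
; The function returns two booleans and three floats, all false or zero in the interface stub, with the logic left to implementing classes.
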(fn FiringAbilityCooldownOverride
  (return false 0.0 0.0 0.0 false))
